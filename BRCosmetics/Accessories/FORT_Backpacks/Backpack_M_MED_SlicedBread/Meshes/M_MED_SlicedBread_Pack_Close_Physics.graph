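# Physics asset for the closed-flap SlicedBread backpack cosmetic, giving its flap and small straps capsule bodies for secondary motion.
# Two large capsules form the pack body, which joins the root with free swing and twist.
# The flap hinges off the pack on one swing axis and carries left and right three-segment strap chains.
# Strap swing limits widen down each chain, while twist stays almost locked.
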PhysicsAsset /BRCosmetics/Accessories/FORT_Backpacks/Backpack_M_MED_SlicedBread/Meshes/M_MED_SlicedBread_Pack_Close_Physics
bodies {
 dyn_flap_closed [{
   shapeType=Capsule
   length=24.6353
   radius=2.3
   shapeName=dyn_flap_closed_capsule
   center=<1.7716 0.15 -3.7449>
   rotation=<rot 0 -0.4427 -89.8627>
  }]
 dyn_little_strap_l_00 [{shapeType=Capsule, length=1, radius=2, shapeName=dyn_little_strap_l_00_capsule, center=<2.3665 0 2>, rotation=<rot 90 0 0>}]
 dyn_little_strap_l_01 [{shapeType=Capsule, length=0.1, radius=2, shapeName=dyn_little_strap_l_01_capsule, center=<1.7003 0 2>, rotation=<rot 90 0 0>}]
 dyn_little_strap_l_02 [{shapeType=Capsule, length=1, radius=2, shapeName=dyn_little_strap_l_02_capsule, center=<2.3665 0 2>, rotation=<rot 90 0 0>}]
 dyn_little_strap_r_00 [{shapeType=Capsule, length=1, radius=2, shapeName=dyn_little_strap_r_00_capsule, center=<2.3665 0 1>, rotation=<rot 90 0 0>}]
 dyn_little_strap_r_01 [{shapeType=Capsule, length=0.1, radius=2, shapeName=dyn_little_strap_r_01_capsule, center=<1.7003 0 1.5>, rotation=<rot 90 0 0>}]
 dyn_little_strap_r_02 [{shapeType=Capsule, length=1, radius=2, shapeName=dyn_little_strap_r_02_capsule, center=<2.3665 0 1.5>, rotation=<rot 90 0 0>}]
 pack [
  {
   shapeType=Capsule
   length=20.2188
   radius=8.9892
   shapeName=pack_capsule
   center=<-0.3421 6.3323 -1.7402>
   rotation=<rot -0.0004 -88.5663 -86.9418>
  }
  {
   shapeType=Capsule
   length=20.2188
   radius=8.9892
   shapeName=pack_capsule
   center=<-0.6538 5.0272 -8.3943>
   rotation=<rot -0.0007 -88.5663 -86.9418>
  }
 ]
 root [{shapeType=Capsule, length=1, radius=1, shapeName=root_capsule, center=<0.357 1.564 2.8967>, rotation=<rot -2.8217 -90 -13.7681>}]
}
constraints [
 {
  bone1Name=pack
  bone2Name=root
  swing1LimitDegrees=45
  swing1Motion=Free
  swing2LimitDegrees=45
  swing2Motion=Free
  twistLimitDegrees=45
  twistMotion=Free
 }
 {bone1Name=dyn_flap_closed, bone2Name=pack, swing1LimitDegrees=0, swing2LimitDegrees=35, twistLimitDegrees=0}
 {bone1Name=dyn_little_strap_l_00, bone2Name=dyn_flap_closed, swing1LimitDegrees=13, swing2LimitDegrees=40, twistLimitDegrees=1}
 {bone1Name=dyn_little_strap_l_01, bone2Name=dyn_little_strap_l_00, swing1LimitDegrees=13, swing2LimitDegrees=60, twistLimitDegrees=1}
 {bone1Name=dyn_little_strap_l_02, bone2Name=dyn_little_strap_l_01, swing1LimitDegrees=13, swing2LimitDegrees=65, twistLimitDegrees=1}
 {bone1Name=dyn_little_strap_r_00, bone2Name=dyn_flap_closed, swing1LimitDegrees=13, swing2LimitDegrees=40, twistLimitDegrees=1}
 {bone1Name=dyn_little_strap_r_01, bone2Name=dyn_little_strap_r_00, swing1LimitDegrees=13, swing2LimitDegrees=60, twistLimitDegrees=1}
 {bone1Name=dyn_little_strap_r_02, bone2Name=dyn_little_strap_r_01, swing1LimitDegrees=13, swing2LimitDegrees=65, twistLimitDegrees=1}
]
