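# Physics asset for the open variant of the SlicedBread backpack cosmetic, defining capsule bodies and joint limits for its moving parts.
# Two capsules form the pack body, to which a small bag body and a two-capsule flap are constrained.
# Left and right three-segment front strap chains hang from the flap.
# Tight twist limits and wider second-swing limits on the flap and straps make them bend mainly along one axis.
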PhysicsAsset /BRCosmetics/Accessories/FORT_Backpacks/Backpack_M_MED_SlicedBread/Meshes/M_MED_SlicedBread_Pack_Open_Physics
bodies {
 dyn_bag [{shapeType=Capsule, length=1, radius=1, shapeName=dyn_bag_capsule}]
 dyn_flap_opened [
  {
   shapeType=Capsule
   length=25.1353
   radius=2.8
   shapeName=dyn_flap_closed_capsule
   center=<3.0234 0.1577 3.4994>
   rotation=<rot 0 -0.4427 -89.8627>
  }
  {
   shapeType=Capsule
   length=14.6353
   radius=2.8
   shapeName=dyn_flap_opened_capsule
   center=<6.3921 -0.4774 3.5011>
   rotation=<rot 0 -0.4427 -89.8627>
  }
 ]
 dyn_l_strap_front_00 [{shapeType=Capsule, length=2, radius=1.8, shapeName=dyn_little_strap_l_00_capsule, center=<2.8805 0 0>, rotation=<rot 90 0 0>}]
 dyn_l_strap_front_01 [{shapeType=Capsule, length=0, radius=1.8, shapeName=dyn_l_strap_front_01_capsule, center=<1.7924 0 0>, rotation=<rot 90 0 0>}]
 dyn_l_strap_front_02 [{shapeType=Capsule, length=2, radius=1.8, shapeName=dyn_l_strap_front_02_capsule, center=<2.7954 0 0>, rotation=<rot 90 0 0>}]
 dyn_r_strap_front_00 [{shapeType=Capsule, length=2, radius=1, shapeName=dyn_little_strap_l_00_capsule, center=<2.8805 0 1>, rotation=<rot 90 0 0>}]
 dyn_r_strap_front_01 [{shapeType=Capsule, length=0, radius=1.5, shapeName=dyn_r_strap_front_01_capsule, center=<1.8069 0 1>, rotation=<rot 90 0 0>}]
 dyn_r_strap_front_02 [{shapeType=Capsule, length=2, radius=1.8, shapeName=dyn_r_strap_front_02_capsule, center=<2.6166 0 1>, rotation=<rot 90 0 0>}]
 pack [
  {
   shapeType=Capsule
   length=24.7188
   radius=6.9892
   shapeName=pack_capsule
   center=<-0.0906 2.5458 1.193>
   rotation=<rot -0.0004 -88.5663 -86.9418>
  }
  {
   shapeType=Capsule
   length=23.7188
   radius=7.9892
   shapeName=pack_capsule
   center=<-0.4963 3.1689 -7.3805>
   rotation=<rot -0.0004 -88.5663 -86.9418>
  }
 ]
}
constraints [
 {bone1Name=dyn_l_strap_front_01, bone2Name=dyn_l_strap_front_00, swing1LimitDegrees=15, swing2LimitDegrees=60, twistLimitDegrees=1}
 {bone1Name=dyn_l_strap_front_02, bone2Name=dyn_l_strap_front_01, swing1LimitDegrees=15, swing2LimitDegrees=60, twistLimitDegrees=1}
 {bone1Name=dyn_r_strap_front_01, bone2Name=dyn_r_strap_front_00, swing1LimitDegrees=13, swing2LimitDegrees=80, twistLimitDegrees=1}
 {bone1Name=dyn_r_strap_front_02, bone2Name=dyn_r_strap_front_01, swing1LimitDegrees=13, swing2LimitDegrees=80, twistLimitDegrees=1}
 {bone1Name=dyn_bag, bone2Name=pack, swing1LimitDegrees=1, swing2LimitDegrees=2, twistLimitDegrees=1}
 {bone1Name=dyn_l_strap_front_00, bone2Name=dyn_flap_opened, swing1LimitDegrees=15, swing2LimitDegrees=60, twistLimitDegrees=1}
 {bone1Name=dyn_r_strap_front_00, bone2Name=dyn_flap_opened, swing1LimitDegrees=13, swing2LimitDegrees=60, twistLimitDegrees=1}
 {bone1Name=dyn_flap_opened, bone2Name=pack, swing1LimitDegrees=1, swing2LimitDegrees=30, twistLimitDegrees=0}
]
