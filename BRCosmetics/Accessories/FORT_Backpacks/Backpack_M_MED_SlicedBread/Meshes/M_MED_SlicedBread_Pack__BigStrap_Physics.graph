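# Physics asset for the big-strap physics of the SlicedBread backpack cosmetic, giving its dangling strap bones collision and constrained motion.
# Capsule bodies cover three dynamic strap bones, and two large capsules approximate the pack.
# Constraints chain pack to strap 01, 02 and 03 in sequence.
# Each joint allows wide swing on one axis, much narrower swing on the other and almost no twist, so the strap bends mainly in one plane.
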PhysicsAsset /BRCosmetics/Accessories/FORT_Backpacks/Backpack_M_MED_SlicedBread/Meshes/M_MED_SlicedBread_Pack__BigStrap_Physics
bodies {
 dyn_strap_01 [{shapeType=Capsule, length=1, radius=3, shapeName=dyn_strap_01_capsule, center=<3.9204 0 1>, rotation=<rot 90 0 0>}]
 dyn_strap_02 [{shapeType=Capsule, length=5.8, radius=3, center=<6.0859 0 0.1608>, rotation=<rot 90 0 0>}]
 dyn_strap_03 [{shapeType=Capsule, length=19, radius=5.5, shapeName=dyn_strap_03_capsule, center=<5.4245 0 1.2697>, rotation=<rot 0 0 90>}]
 pack [
  {shapeType=Capsule, length=39, radius=27.5, center=<-0.0001 30.055 -15.4235>, rotation=<rot 90 180 179.9989>}
  {shapeType=Capsule, length=39, radius=27.5, shapeName=pack_capsule, center=<0.0942 28.8715 -37.5206>, rotation=<rot 90 -179.8422 -179.8432>}
 ]
}
constraints [
 {bone1Name=dyn_strap_02, bone2Name=dyn_strap_01, swing1LimitDegrees=10, swing2LimitDegrees=60, twistLimitDegrees=2}
 {bone1Name=dyn_strap_01, bone2Name=pack, swing1LimitDegrees=5, swing2LimitDegrees=50, twistLimitDegrees=2}
 {bone1Name=dyn_strap_03, bone2Name=dyn_strap_02, swing1LimitDegrees=20, swing2LimitDegrees=60, twistLimitDegrees=2}
]
