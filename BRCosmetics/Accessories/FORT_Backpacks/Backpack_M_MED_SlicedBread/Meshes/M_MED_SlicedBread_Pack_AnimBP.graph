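# Anim Blueprint for the SlicedBread backpack, adding secondary motion and swapping material states with the matching pickaxe.
# Copies the mesh pose, then blends by bRigidBodyEnabled between RigidBody/AnimDynamics and cheaper Trail controller chains.
# Both branches offset pack and straps for skydive and parachute; several other node chains are disconnected.
# The EventGraph updates gravity override and, on weapon, skydive or customization events, toggles element hiding and vertex crunch parameters.
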AnimBlueprint /BRCosmetics/Accessories/FORT_Backpacks/Backpack_M_MED_SlicedBread/Meshes/M_MED_SlicedBread_Pack_AnimBP
target_skeleton=&/M_MED_SlicedBread_Pack_Skeleton
graphs [
 AnimGraph {nodes [
   n1 = Get bIsSkydiveDiveMode()
   n2 = "Transform (Modify) Bone - Bone: pack"!(ComponentPose, bAlphaBoolEnabled:bool=@n1, Translation="0,0,0", Rotation="0,0,15", Scale) -> (Pose) {
    alphaInputType=Bool
    translationMode=BMM_Additive
    alphaBoolBlend {blendOption=Linear}
    alphaScaleBiasClamp {bInterpResult=true}
    boneToModify {boneName=pack}
    rotation {roll=15}
   }
   n3 = "Transform (Modify) Bone - Bone: dyn_strap_01"!(ComponentPose=@n2, bAlphaBoolEnabled:bool=@n1, Translation="0,0,-1", Rotation="0.000,0.000,-30.000", Scale="1,1,1") -> (Pose) {
    alphaInputType=Bool
    rotationMode=BMM_Additive
    translationMode=BMM_Additive
    alphaBoolBlend {blendOption=Linear}
    alphaScaleBiasClamp {bInterpResult=true}
    boneToModify {boneName=dyn_strap_01}
    rotation {roll=-30}
    translation {z=-1}
   }
   n4 = "Transform (Modify) Bone - Bone: dyn_strap_02"!(ComponentPose=@n3, bAlphaBoolEnabled:bool=@n1, Translation="0,0,-1", Rotation="0.000,0.000,10.000", Scale="1,1,1") -> (Pose) {
    alphaInputType=Bool
    rotationMode=BMM_Additive
    translationMode=BMM_Additive
    alphaBoolBlend {blendOption=Linear}
    alphaScaleBiasClamp {bInterpResult=true}
    boneToModify {boneName=dyn_strap_02}
    rotation {roll=10}
    translation {z=-1}
   }
   n5 = "Transform (Modify) Bone - Bone: dyn_strap_03"!(ComponentPose=@n4, bAlphaBoolEnabled:bool=@n1, Translation="0,0,-1", Rotation="0.000,0.000,20.000", Scale="1,1,1") -> (Pose) {
    alphaInputType=Bool
    rotationMode=BMM_Additive
    translationMode=BMM_Additive
    alphaBoolBlend {blendOption=Linear}
    alphaScaleBiasClamp {bInterpResult=true}
    boneToModify {boneName=dyn_strap_03}
    rotation {roll=20}
    translation {z=-1}
   }
   Anim Dynamics - Bone: dyn_trophie!(ComponentPose=@n5, bAlphaBoolEnabled:bool=@n1, ExternalForce="0,0,0") -> (Pose) {
    alphaInputType=Bool
    angularDampingOverride=0.8
    angularSpringConstant=1000
    bAngularSpring=true
    bOverrideAngularDamping=true
    bUseGravityOverride=true
    bUsePlanarLimit=false
    simulationSpace=RootRelative
    alphaBoolBlend {blendOption=Linear}
    boundBone {boneName=dyn_trophie}
    componentAppliedLinearAccClamp=<1000 1000 1000>
    componentLinearAccScale=<0.5 0.5 0.5>
    gravityOverride {z=-980}
    physicsBodyDefinitions [{
      boundBone {boneName=dyn_trophie}
      boxExtents=<10 10 10>
      constraintSetup {angularLimitsMax=<1 6 1>, angularLimitsMin=<-1 -6 -1>, angularTarget=<0 0 0>}
      localJointOffset=<2 0.0517 -0.0157>
     }]
    relativeSpaceBone {boneName=pack}
   }
   n6 = Anim Dynamics - Bone: dyn_strap_01!(ComponentPose, bAlphaBoolEnabled:bool=True, ExternalForce="0,0,0") -> (Pose) {
    alphaInputType=Bool
    angularDampingOverride=0.9999
    angularSpringConstant=5000
    bAngularSpring=true
    bOverrideAngularDamping=true
    bUseGravityOverride=true
    bUsePlanarLimit=false
    simulationSpace=RootRelative
    alphaBoolBlend {blendOption=Linear}
    boundBone {boneName=dyn_strap_01}
    componentAppliedLinearAccClamp=<1000 1000 1000>
    componentLinearAccScale=<0.5 0.5 0.8>
    gravityOverride {z=-980}
    physicsBodyDefinitions [{
      boundBone {boneName=dyn_strap_01}
      boxExtents=<10 10 10>
      constraintSetup {angularLimitsMax=<0 0 0>, angularLimitsMin=<0 0 0>, angularTarget=<0 0 0>}
      localJointOffset=<2 0.0517 -0.0157>
     }]
    relativeSpaceBone {boneName=pack}
   }
   n7 = Anim Dynamics - Bone: dyn_strap_02!(ComponentPose=@n6, bAlphaBoolEnabled:bool=True, ExternalForce="0,0,0") -> (Pose) {
    alphaInputType=Bool
    angularDampingOverride=0.9999
    angularSpringConstant=5000
    bAngularSpring=true
    bOverrideAngularDamping=true
    bUseGravityOverride=true
    bUsePlanarLimit=false
    simulationSpace=RootRelative
    alphaBoolBlend {blendOption=Linear}
    boundBone {boneName=dyn_strap_02}
    componentAppliedLinearAccClamp=<1000 1000 1000>
    componentLinearAccScale=<0.5 0.5 0.8>
    gravityOverride {z=-980}
    physicsBodyDefinitions [{
      boundBone {boneName=dyn_strap_02}
      boxExtents=<10 10 10>
      constraintSetup {angularLimitsMax=<0 0 0>, angularLimitsMin=<0 0 0>, angularTarget=<0 0 0>}
      localJointOffset=<2 0.0517 -0.0157>
     }]
    relativeSpaceBone {boneName=pack}
   }
   Anim Dynamics - Bone: dyn_strap_03!(ComponentPose=@n7, bAlphaBoolEnabled:bool=True, ExternalForce="0,0,0") -> (Pose) {
    alphaInputType=Bool
    angularDampingOverride=0.9999
    angularSpringConstant=5000
    bAngularSpring=true
    bOverrideAngularDamping=true
    bUseGravityOverride=true
    bUsePlanarLimit=false
    simulationSpace=RootRelative
    alphaBoolBlend {blendOption=Linear}
    boundBone {boneName=dyn_strap_03}
    componentAppliedLinearAccClamp=<1000 1000 1000>
    componentLinearAccScale=<0.5 0.5 0.8>
    gravityOverride {z=-980}
    physicsBodyDefinitions [{
      boundBone {boneName=dyn_strap_03}
      boxExtents=<10 10 10>
      constraintSetup {angularLimitsMax=<0 0 0>, angularLimitsMin=<0 0 0>, angularTarget=<0 0 0>}
      localJointOffset=<2 0.0517 -0.0157>
     }]
    relativeSpaceBone {boneName=pack}
   }
   n8 = Anim Dynamics - Bone: dyn_strap_01(bAlphaBoolEnabled=True, ExternalForce="0,0,0") {
    alphaInputType=Bool
    angularDampingOverride=0.9999
    angularSpringConstant=5000
    bAngularSpring=true
    bOverrideAngularDamping=true
    bUseGravityOverride=true
    bUsePlanarLimit=false
    bUseSphericalLimits=true
    simulationSpace=RootRelative
    alphaBoolBlend {blendOption=Linear}
    boundBone {boneName=dyn_strap_01}
    componentAppliedLinearAccClamp=<1000 1000 1000>
    componentLinearAccScale=<0.5 0.5 0.8>
    gravityOverride {z=-980}
    physicsBodyDefinitions [{
      boundBone {boneName=dyn_strap_01}
      boxExtents=<10 10 10>
      constraintSetup {angularLimitsMax=<10 -4.5 10>, angularLimitsMin=<-10 -80 -10>, angularTarget=<0 0 0>}
      localJointOffset=<2 0.0517 -0.0157>
     }]
    relativeSpaceBone {boneName=pack}
    sphericalLimits [{limitRadius=30, drivingBone {boneName=root}, sphereLocalOffset=<0 29.9865 0>}]
   }
   n9 = Anim Dynamics - Bone: dyn_strap_02(ComponentPose=@n8, bAlphaBoolEnabled=True, ExternalForce="0,0,0") {
    alphaInputType=Bool
    angularDampingOverride=0.9999
    angularSpringConstant=5000
    bAngularSpring=true
    bOverrideAngularDamping=true
    bUseGravityOverride=true
    bUsePlanarLimit=false
    bUseSphericalLimits=true
    simulationSpace=RootRelative
    alphaBoolBlend {blendOption=Linear}
    boundBone {boneName=dyn_strap_02}
    componentAppliedLinearAccClamp=<1000 1000 1000>
    componentLinearAccScale=<0.5 0.5 0.8>
    gravityOverride {z=-980}
    physicsBodyDefinitions [{
      boundBone {boneName=dyn_strap_02}
      boxExtents=<10 10 10>
      constraintSetup {angularLimitsMax=<10 0 10>, angularLimitsMin=<-10 -50 -10>, angularTarget=<0 0 0>}
      localJointOffset=<2 0.0517 -0.0157>
     }]
    relativeSpaceBone {boneName=pack}
    sphericalLimits [{limitRadius=30, drivingBone {boneName=root}, sphereLocalOffset=<0 29.9628 -12.2632>}]
   }
   Anim Dynamics - Bone: dyn_strap_03(ComponentPose=@n9, bAlphaBoolEnabled=True, ExternalForce="0,0,0") {
    alphaInputType=Bool
    angularDampingOverride=0.9999
    angularSpringConstant=5000
    bAngularSpring=true
    bOverrideAngularDamping=true
    bUseGravityOverride=true
    bUsePlanarLimit=false
    bUseSphericalLimits=true
    simulationSpace=RootRelative
    alphaBoolBlend {blendOption=Linear}
    boundBone {boneName=dyn_strap_03}
    componentAppliedLinearAccClamp=<1000 1000 1000>
    componentLinearAccScale=<0.5 0.5 0.8>
    gravityOverride {z=-980}
    physicsBodyDefinitions [{
      boundBone {boneName=dyn_strap_03}
      boxExtents=<10 10 10>
      constraintSetup {angularLimitsMax=<10 0 10>, angularLimitsMin=<-10 -50 -10>, angularTarget=<0 0 0>}
      localJointOffset=<2 0.0517 -0.0157>
     }]
    relativeSpaceBone {boneName=pack}
    sphericalLimits [{limitRadius=30, drivingBone {boneName=root}, sphereLocalOffset=<0 46.4272 -38.6485>}]
   }
   n10 = Anim Dynamics - Bone: dyn_strap_03!(ComponentPose, Alpha:float=1, ExternalForce="0,0,0") -> (Pose) {
    angularDampingOverride=0.99
    angularSpringConstant=2000
    bAngularSpring=true
    bGravityOverrideInSimSpace=true
    bOverrideAngularDamping=true
    bOverrideLinearDamping=true
    bUseGravityOverride=true
    gravityScale=0.35
    linearDampingOverride=0.99
    simulationSpace=BoneRelative
    alphaBoolBlend {blendOption=Linear}
    alphaScaleBiasClamp {scale=0.5}
    boundBone {boneName=dyn_strap_03}
    componentAppliedLinearAccClamp=<5000 5000 5000>
    componentLinearAccScale=<0.3 0.3 0.4>
    componentLinearVelScale=<0.3 0.3 0.4>
    gravityOverride {z=500}
    physicsBodyDefinitions [{
      sphereCollisionRadius=0
      boundBone {boneName=dyn_strap_03}
      boxExtents=<10 10 10>
      constraintSetup {angularLimitsMax=<10 10 10>, angularLimitsMin=<-30 -30 -30>, angularTarget=<1 0 0>}
      localJointOffset=<1 0 0>
     }]
   }
   n11 = Anim Dynamics - Bone: dyn_strap_02!(ComponentPose=@n10, Alpha:float=1, ExternalForce="0,0,0") -> (Pose) {
    angularDampingOverride=0.99
    angularSpringConstant=2000
    bAngularSpring=true
    bGravityOverrideInSimSpace=true
    bOverrideAngularDamping=true
    bOverrideLinearDamping=true
    bUseGravityOverride=true
    gravityScale=0.35
    linearDampingOverride=0.99
    simulationSpace=BoneRelative
    alphaBoolBlend {blendOption=Linear}
    alphaScaleBiasClamp {scale=0.5}
    boundBone {boneName=dyn_strap_02}
    componentAppliedLinearAccClamp=<5000 5000 5000>
    componentLinearAccScale=<0.3 0.3 0.4>
    componentLinearVelScale=<0.3 0.3 0.4>
    gravityOverride {z=500}
    physicsBodyDefinitions [{
      sphereCollisionRadius=0
      boundBone {boneName=dyn_strap_02}
      boxExtents=<10 10 10>
      constraintSetup {angularLimitsMax=<10 10 10>, angularLimitsMin=<-30 -30 -30>, angularTarget=<1 0 0>}
      localJointOffset=<1 0 0>
     }]
   }
   n12 = Anim Dynamics - Bone: dyn_strap_05(ComponentPose=@n11, Alpha=1, ExternalForce="0,0,0") {
    angularDampingOverride=0.99
    angularSpringConstant=2000
    bAngularSpring=true
    bGravityOverrideInSimSpace=true
    bOverrideAngularDamping=true
    bOverrideLinearDamping=true
    bUseGravityOverride=true
    gravityScale=0.35
    linearDampingOverride=0.99
    simulationSpace=BoneRelative
    alphaBoolBlend {blendOption=Linear}
    alphaScaleBiasClamp {scale=0.5}
    boundBone {boneName=dyn_strap_05}
    componentAppliedLinearAccClamp=<5000 5000 5000>
    componentLinearAccScale=<0.3 0.3 0.4>
    componentLinearVelScale=<0.3 0.3 0.4>
    gravityOverride {z=500}
    physicsBodyDefinitions [{
      sphereCollisionRadius=0
      boundBone {boneName=dyn_strap_05}
      boxExtents=<10 10 10>
      constraintSetup {angularLimitsMax=<10 10 10>, angularLimitsMin=<-30 -30 -30>, angularTarget=<1 0 0>}
      localJointOffset=<1 0 0>
     }]
   }
   n13 = Anim Dynamics - Bone: dyn_strap_06(ComponentPose=@n12, Alpha=1, ExternalForce="0,0,0") {
    angularDampingOverride=0.99
    angularSpringConstant=2000
    bAngularSpring=true
    bGravityOverrideInSimSpace=true
    bOverrideAngularDamping=true
    bOverrideLinearDamping=true
    bUseGravityOverride=true
    gravityScale=0.35
    linearDampingOverride=0.99
    simulationSpace=BoneRelative
    alphaBoolBlend {blendOption=Linear}
    alphaScaleBiasClamp {scale=0.5}
    boundBone {boneName=dyn_strap_06}
    componentAppliedLinearAccClamp=<5000 5000 5000>
    componentLinearAccScale=<0.3 0.3 0.4>
    componentLinearVelScale=<0.3 0.3 0.4>
    gravityOverride {z=500}
    physicsBodyDefinitions [{
      sphereCollisionRadius=0
      boundBone {boneName=dyn_strap_06}
      boxExtents=<10 10 10>
      constraintSetup {angularLimitsMax=<10 10 10>, angularLimitsMin=<-30 -30 -30>, angularTarget=<1 0 0>}
      localJointOffset=<1 0 0>
     }]
   }
   Anim Dynamics - Bone: dyn_strap_01!(ComponentPose=@n13, Alpha:float=1, ExternalForce="0,0,0") -> (Pose) {
    angularDampingOverride=0.99
    angularSpringConstant=2000
    bAngularSpring=true
    bGravityOverrideInSimSpace=true
    bOverrideAngularDamping=true
    bOverrideLinearDamping=true
    bUseGravityOverride=true
    gravityScale=0.35
    linearDampingOverride=0.99
    simulationSpace=BoneRelative
    alphaBoolBlend {blendOption=Linear}
    alphaScaleBiasClamp {scale=0.5}
    boundBone {boneName=dyn_strap_01}
    componentAppliedLinearAccClamp=<5000 5000 5000>
    componentLinearAccScale=<0.3 0.3 0.4>
    componentLinearVelScale=<0.3 0.3 0.4>
    gravityOverride {z=500}
    physicsBodyDefinitions [{
      sphereCollisionRadius=0
      boundBone {boneName=dyn_strap_01}
      boxExtents=<10 10 10>
      constraintSetup {angularLimitsMax=<10 10 10>, angularLimitsMin=<-30 -30 -30>, angularTarget=<1 0 0>}
      localJointOffset=<1 0 0>
     }]
   }
   Input Pose()
   "Transform (Modify) Bone - Bone: pack"!(ComponentPose, Alpha:float, Translation="0.000,0.038,-1.414", Rotation="0,0,0", Scale) -> (Pose) {translationMode=BMM_Additive, alphaBoolBlend {blendOption=Linear}, boneToModify {boneName=pack}, translation {y=0.038, z=-1.414}}
   n14 = Get bIsMoving2D()
   n15 = "Transform (Modify) Bone - Bone: dyn_strap_01"(bAlphaBoolEnabled=@n14, Translation="0,0,-1", Rotation="0,0,10", Scale="1,1,1") {
    alphaInputType=Bool
    rotationMode=BMM_Additive
    translationMode=BMM_Additive
    alphaBoolBlend {blendOption=Linear, blendOutTime=1}
    alphaScaleBiasClamp {bInterpResult=true}
    boneToModify {boneName=dyn_strap_01}
    rotation {roll=10}
    translation {z=-1}
   }
   n16 = "Transform (Modify) Bone - Bone: dyn_strap_02"(ComponentPose=@n15, bAlphaBoolEnabled=@n14, Translation="0,0,-1", Rotation="0,0,0", Scale="1,1,1") {
    alphaInputType=Bool
    rotationMode=BMM_Additive
    translationMode=BMM_Additive
    alphaBoolBlend {blendOption=Linear, blendOutTime=1}
    alphaScaleBiasClamp {bInterpResult=true}
    boneToModify {boneName=dyn_strap_02}
    translation {z=-1}
   }
   "Transform (Modify) Bone - Bone: dyn_strap_03"(ComponentPose=@n16, bAlphaBoolEnabled=@n14, Translation="0,0,-1", Rotation="0.000,0.000,10.000", Scale="1,1,1") {
    alphaInputType=Bool
    rotationMode=BMM_Additive
    translationMode=BMM_Additive
    alphaBoolBlend {blendOption=Linear, blendOutTime=1}
    alphaScaleBiasClamp {bInterpResult=true}
    boneToModify {boneName=dyn_strap_03}
    rotation {roll=10}
    translation {z=-1}
   }
   n17 = "Transform (Modify) Bone - Bone: dyn_strap_05"(Alpha=1, Translation="0,-5,2", Rotation="0,30,0", Scale="1,1,1") {
    rotationMode=BMM_Additive
    translationMode=BMM_Additive
    alphaBoolBlend {blendOption=Linear}
    boneToModify {boneName=dyn_strap_05}
    rotation {yaw=30}
    translation {y=-5, z=2}
   }
   n18 = "Transform (Modify) Bone - Bone: dyn_strap_04"(ComponentPose=@n17, Alpha=1, Translation="0,-2.5,3.5", Rotation="-0.000,-20.000,0.000", Scale="1,1,1") {
    rotationMode=BMM_Additive
    translationMode=BMM_Additive
    alphaBoolBlend {blendOption=Linear}
    boneToModify {boneName=dyn_strap_04}
    rotation {yaw=-20}
    translation {y=-2.5, z=3.5}
   }
   n19 = "Transform (Modify) Bone - Bone: dyn_strap_03"!(ComponentPose=@n18, Alpha:float=1, Translation="0,-2,3", Rotation="0.000,10.000,-0.000", Scale="1,1,1") -> (Pose) {
    rotationMode=BMM_Additive
    translationMode=BMM_Additive
    alphaBoolBlend {blendOption=Linear}
    boneToModify {boneName=dyn_strap_03}
    rotation {yaw=10}
    translation {y=-2, z=3}
   }
   n20 = "Transform (Modify) Bone - Bone: dyn_strap_01"!(ComponentPose=@n19, Alpha:float=1, Translation="0,0,0", Rotation="10,-0.000,10", Scale="1,1,1") -> (Pose) {
    rotationMode=BMM_Additive
    translationMode=BMM_Additive
    alphaBoolBlend {blendOption=Linear}
    boneToModify {boneName=dyn_strap_01}
    rotation {pitch=10, roll=10}
   }
   "Transform (Modify) Bone - Bone: dyn_strap_02"!(ComponentPose=@n20, Alpha:float=1, Translation="0,0,0", Rotation="4,0.000,10.000", Scale="1,1,1") -> (Pose) {
    rotationMode=BMM_Additive
    translationMode=BMM_Additive
    alphaBoolBlend {blendOption=Linear}
    boneToModify {boneName=dyn_strap_02}
    rotation {pitch=4, roll=10}
   }
   n21 = "Transform (Modify) Bone - Bone: pack"!(
    ComponentPose=Local To Component(LocalPose=Copy Pose From Mesh())
    bAlphaBoolEnabled:bool
    Translation="0.000,-1.587,3.000"
    Rotation="0,0,0"
    Scale
   ) -> (Pose) {
    alphaInputType=Bool
    rotationMode=BMM_Additive
    translationMode=BMM_Additive
    alphaBoolBlend {blendOption=Linear}
    boneToModify {boneName=pack}
    translation {y=-1.587, z=3}
   }
   n22 = "Transform (Modify) Bone - Bone: dyn_strap_02"!(ComponentPose=@n21, bAlphaBoolEnabled:bool=True, Translation="0,0,0", Rotation="0,0,25", Scale="1,1,1") -> (Pose) {
    alphaInputType=Bool
    rotationMode=BMM_Additive
    translationMode=BMM_Additive
    alphaBoolBlend {blendOption=Linear}
    boneToModify {boneName=dyn_strap_02}
    rotation {roll=25}
   }
   n23 = RigidBody!(ComponentPose=@n22, Alpha, OverrideWorldGravity:Vector=Get GravityOverride(), ExternalForce) -> (Pose) {
    bClampLinearTranslationLimitToRefPose=true
    bOverrideWorldGravity=true
    overridePhysicsAsset=&/M_MED_SlicedBread_Pack_Open_Physics
    simulationSpace=BaseBoneSpace
    alphaBoolBlend {blendOption=Linear}
    baseBoneRef {boneName=root}
    componentAppliedLinearAccClamp=<5000 5000 5000>
    componentLinearAccScale=<0.45 0.45 0.5>
    overrideWorldGravity {z=-980}
    simSpaceSettings {
     maxAngularAcceleration=1000
     maxAngularVelocity=3
     maxLinearAcceleration=700
     maxLinearVelocity=500
     velocityScaleZ=0
     worldAlpha=1
     externalLinearDragV=<0.2 0.2>
    }
   }
   n24 = RigidBody(ComponentPose=@n23, OverrideWorldGravity=Get GravityOverride()) {
    bClampLinearTranslationLimitToRefPose=true
    bOverrideWorldGravity=true
    overridePhysicsAsset=&/M_MED_SlicedBread_Pack_Close_Physics
    simulationSpace=BaseBoneSpace
    alphaBoolBlend {blendOption=Linear}
    baseBoneRef {boneName=root}
    componentAppliedLinearAccClamp=<5000 5000 5000>
    componentLinearAccScale=<0.2 0.2 0.6>
    overrideWorldGravity {z=-980}
    simSpaceSettings {
     maxAngularAcceleration=1000
     maxAngularVelocity=3
     maxLinearAcceleration=700
     maxLinearVelocity=500
     velocityScaleZ=0
     worldAlpha=0.5
     externalLinearDragV=<0.2 0.2>
    }
   }
   n25 = RigidBody(ComponentPose=@n24, OverrideWorldGravity=Get GravityOverride()) {
    bClampLinearTranslationLimitToRefPose=true
    bOverrideWorldGravity=true
    overridePhysicsAsset=&/M_MED_SlicedBread_Pack__BigStrap_Physics
    simulationSpace=BaseBoneSpace
    alphaBoolBlend {blendOption=Linear}
    baseBoneRef {boneName=root}
    componentAppliedLinearAccClamp=<5000 5000 5000>
    componentLinearAccScale=<0.3 0.3 0.6>
    overrideWorldGravity {z=-980}
    simSpaceSettings {
     maxAngularAcceleration=1000
     maxAngularVelocity=3
     maxLinearAcceleration=700
     maxLinearVelocity=500
     velocityScaleZ=0
     worldAlpha=0.5
     externalLinearDragV=<0.2 0.2>
    }
   }
   n26 = Anim Dynamics - Bone: dyn_trophie!(ComponentPose=@n25, Alpha:float=0.65, ExternalForce="0,0,0") -> (Pose) {
    angularDampingOverride=0.8
    angularSpringConstant=1000
    bAngularSpring=true
    bOverrideAngularDamping=true
    bUseGravityOverride=true
    bUsePlanarLimit=false
    simulationSpace=RootRelative
    alphaBoolBlend {blendOption=Linear}
    boundBone {boneName=dyn_trophie}
    componentAppliedLinearAccClamp=<1000 1000 1000>
    componentLinearAccScale=<0.5 0.5 0.5>
    gravityOverride {z=-980}
    physicsBodyDefinitions [{
      boundBone {boneName=dyn_trophie}
      boxExtents=<10 10 10>
      constraintSetup {angularLimitsMax=<1 6 1>, angularLimitsMin=<-1 -6 -1>, angularTarget=<0 0 0>}
      localJointOffset=<2 0.0517 -0.0157>
     }]
    relativeSpaceBone {boneName=pack}
   }
   n27 = Anim Dynamics - Bone: dyn_metal_buckle_l(ComponentPose=@n26, Alpha=1, ExternalForce="0,0,0") {
    angularDampingOverride=0.8
    angularSpringConstant=1000
    bAngularSpring=true
    bOverrideAngularDamping=true
    bUseGravityOverride=true
    bUsePlanarLimit=false
    simulationSpace=RootRelative
    alphaBoolBlend {blendOption=Linear}
    boundBone {boneName=dyn_metal_buckle_l}
    componentAppliedLinearAccClamp=<1000 1000 1000>
    componentLinearAccScale=<0.5 0.5 0.8>
    gravityOverride {z=-980}
    physicsBodyDefinitions [{
      boundBone {boneName=dyn_metal_buckle_l}
      boxExtents=<10 10 10>
      constraintSetup {angularLimitsMax=<0 10 0>, angularLimitsMin=<0 -60 0>, angularTarget=<0 0 0>}
      localJointOffset=<2 0.0517 -0.0157>
     }]
    relativeSpaceBone {boneName=pack}
   }
   n28 = Anim Dynamics - Bone: dyn_metal_buckle_r(ComponentPose=@n27, Alpha=1, ExternalForce="0,0,0") {
    angularDampingOverride=0.8
    angularSpringConstant=1000
    bAngularSpring=true
    bOverrideAngularDamping=true
    bUseGravityOverride=true
    bUsePlanarLimit=false
    simulationSpace=RootRelative
    alphaBoolBlend {blendOption=Linear}
    boundBone {boneName=dyn_metal_buckle_r}
    componentAppliedLinearAccClamp=<1100 1100 1100>
    componentLinearAccScale=<0.5 0.5 0.8>
    gravityOverride {z=-980}
    physicsBodyDefinitions [{
      boundBone {boneName=dyn_metal_buckle_r}
      boxExtents=<10 10 10>
      constraintSetup {angularLimitsMax=<0 7 0>, angularLimitsMin=<0 -40 0>, angularTarget=<0 0 0>}
      localJointOffset=<2 0.0517 -0.0157>
     }]
    relativeSpaceBone {boneName=pack}
   }
   n29 = Get Is Skydive Floating() -> (Is Skydive Floating:double)
   n30 = "Transform (Modify) Bone - Bone: pack"!(ComponentPose=@n28, Alpha:float=@n29, Translation="0.000,-1.424,-1.256", Rotation="0.000,0.000,15.000", Scale) -> (Pose) {
    rotationMode=BMM_Additive
    translationMode=BMM_Additive
    alphaBoolBlend {blendOption=Linear}
    alphaScaleBiasClamp {bInterpResult=true}
    boneToModify {boneName=pack}
    rotation {roll=15}
    translation {y=-1.424, z=-1.256}
   }
   n31 = "Transform (Modify) Bone - Bone: dyn_strap_01"(ComponentPose=@n30, Alpha=@n29, Translation="0,0,-1", Rotation="0.000,0.000,-10.000", Scale="1,1,1") {
    rotationMode=BMM_Additive
    translationMode=BMM_Additive
    alphaBoolBlend {blendOption=Linear}
    alphaScaleBiasClamp {bInterpResult=true}
    boneToModify {boneName=dyn_strap_01}
    rotation {roll=-10}
    translation {z=-1}
   }
   n32 = "Transform (Modify) Bone - Bone: dyn_strap_02"!(ComponentPose=@n31, Alpha:float=@n29, Translation="0,0,-1", Rotation="0.000,0.000,-10.000", Scale="1,1,1") -> (Pose) {
    rotationMode=BMM_Additive
    translationMode=BMM_Additive
    alphaBoolBlend {blendOption=Linear}
    alphaScaleBias {scale=2}
    boneToModify {boneName=dyn_strap_02}
    rotation {roll=-10}
    translation {z=-1}
   }
   n33 = "Transform (Modify) Bone - Bone: dyn_strap_03"(ComponentPose=@n32, Alpha=@n29, Translation="0,0,-1", Rotation="0.000,0.000,10.000", Scale="1,1,1") {
    rotationMode=BMM_Additive
    translationMode=BMM_Additive
    alphaBoolBlend {blendOption=Linear}
    alphaScaleBias {scale=2}
    boneToModify {boneName=dyn_strap_03}
    rotation {roll=10}
    translation {z=-1}
   }
   n34 = Anim Dynamics - Bone: dyn_trophie(ComponentPose=@n33, Alpha=@n29, ExternalForce="0,0,0") {
    angularDampingOverride=0.8
    angularSpringConstant=1000
    bAngularSpring=true
    bOverrideAngularDamping=true
    bUseGravityOverride=true
    bUsePlanarLimit=false
    simulationSpace=RootRelative
    alphaBoolBlend {blendOption=Linear}
    alphaScaleBias {scale=0}
    boundBone {boneName=dyn_trophie}
    componentAppliedLinearAccClamp=<1000 1000 1000>
    componentLinearAccScale=<0.5 0.5 0.5>
    gravityOverride {z=-980}
    physicsBodyDefinitions [{
      boundBone {boneName=dyn_trophie}
      boxExtents=<10 10 10>
      constraintSetup {angularLimitsMax=<1 6 1>, angularLimitsMin=<-1 -6 -1>, angularTarget=<0 0 0>}
      localJointOffset=<2 0.0517 -0.0157>
     }]
    relativeSpaceBone {boneName=pack}
   }
   n35 = "Transform (Modify) Bone - Bone: dyn_strap_01"!(ComponentPose=@n34, bAlphaBoolEnabled:bool=Get bIsParachuteOpened(), Translation="0,0,-1", Rotation="0,0,19", Scale="1,1,1") -> (Pose) {
    alphaInputType=Bool
    rotationMode=BMM_Additive
    translationMode=BMM_Additive
    alphaBoolBlend {blendOption=Linear}
    alphaScaleBiasClamp {bInterpResult=true}
    boneToModify {boneName=dyn_strap_01}
    rotation {roll=19}
    translation {z=-1}
   }
   n36 = Get Is Skydive Floating()
   n37 = "Transform (Modify) Bone - Bone: pack"(
    ComponentPose=Local To Component(LocalPose=Copy Pose From Mesh())
    Alpha=@n36
    Translation="0.000,-1.424,-1.256"
    Rotation="0.000,0.000,15.000"
   ) {
    rotationMode=BMM_Additive
    translationMode=BMM_Additive
    alphaBoolBlend {blendOption=Linear}
    alphaScaleBiasClamp {bInterpResult=true}
    boneToModify {boneName=pack}
    rotation {roll=15}
    translation {y=-1.424, z=-1.256}
   }
   n38 = "Transform (Modify) Bone - Bone: dyn_strap_01"!(ComponentPose=@n37, Alpha:float=@n36, Translation="0,0,-1", Rotation="0,0,0", Scale="1,1,1") -> (Pose) {
    rotationMode=BMM_Additive
    translationMode=BMM_Additive
    alphaBoolBlend {blendOption=Linear}
    alphaScaleBiasClamp {bInterpResult=true}
    boneToModify {boneName=dyn_strap_01}
    translation {z=-1}
   }
   n39 = "Transform (Modify) Bone - Bone: dyn_strap_02"(ComponentPose=@n38, Alpha=@n36, Translation="0,0,-1", Rotation="0,0,0", Scale="1,1,1") {
    rotationMode=BMM_Additive
    translationMode=BMM_Additive
    alphaBoolBlend {blendOption=Linear}
    alphaScaleBias {scale=2}
    boneToModify {boneName=dyn_strap_02}
    translation {z=-1}
   }
   n40 = "Transform (Modify) Bone - Bone: dyn_strap_03"(ComponentPose=@n39, Alpha=@n36, Translation="0,0,-1", Rotation="0.000,0.000,20.000", Scale="1,1,1") {
    rotationMode=BMM_Additive
    translationMode=BMM_Additive
    alphaBoolBlend {blendOption=Linear}
    alphaScaleBias {scale=2}
    boneToModify {boneName=dyn_strap_03}
    rotation {roll=20}
    translation {z=-1}
   }
   n41 = Anim Dynamics - Bone: dyn_trophie(ComponentPose=@n40, Alpha=@n36, ExternalForce="0,0,0") {
    angularDampingOverride=0.8
    angularSpringConstant=1000
    bAngularSpring=true
    bOverrideAngularDamping=true
    bUseGravityOverride=true
    bUsePlanarLimit=false
    simulationSpace=RootRelative
    alphaBoolBlend {blendOption=Linear}
    alphaScaleBias {scale=0}
    boundBone {boneName=dyn_trophie}
    componentAppliedLinearAccClamp=<1000 1000 1000>
    componentLinearAccScale=<0.5 0.5 0.5>
    gravityOverride {z=-980}
    physicsBodyDefinitions [{
      boundBone {boneName=dyn_trophie}
      boxExtents=<10 10 10>
      constraintSetup {angularLimitsMax=<1 6 1>, angularLimitsMin=<-1 -6 -1>, angularTarget=<0 0 0>}
      localJointOffset=<2 0.0517 -0.0157>
     }]
    relativeSpaceBone {boneName=pack}
   }
   n42 = "Transform (Modify) Bone - Bone: dyn_strap_01"!(ComponentPose=@n41, bAlphaBoolEnabled:bool=Get bIsParachuteOpened(), Translation="0,0,-1", Rotation="0,0,0", Scale="1,1,1") -> (Pose) {
    alphaInputType=Bool
    rotationMode=BMM_Additive
    translationMode=BMM_Additive
    alphaBoolBlend {blendOption=Linear}
    alphaScaleBiasClamp {bInterpResult=true}
    boneToModify {boneName=dyn_strap_01}
    translation {z=-1}
   }
   n43 = Trail controller - Bone: dyn_strap_03(ComponentPose=@n42, Alpha=0.6) {
    bLimitRotation=true
    bUsePlanarLimit=true
    planarLimits [{drivingBone {boneName=dyn_strap_03}, planeTransform {location=<0 0 0>, rotation=<rot 0 0 0>, scale=<1 1 1>}}]
    rotationLimits [{limitMax=<0 0 0>, limitMin=<0 -60 0>}, {limitMax=<0 0 0>, limitMin=<0 -20 0>}]
    rotationOffsets [<0 0 0>, <0 0 0>]
    trailBone {boneName=dyn_strap_03}
    trailRelaxationSpeed {editorCurveData {keys [<at 0 35 RCIM_Linear>,<at 1 5 RCIM_Linear>]}}
   }
   n44 = Trail controller - Bone: dyn_r_strap_front_02(ComponentPose=@n43, Alpha=1) {
    bLimitRotation=true
    bUsePlanarLimit=true
    chainLength=3
    planarLimits [{drivingBone {boneName=dyn_strap_03}, planeTransform {location=<0 0 0>, rotation=<rot 0 0 0>, scale=<1 1 1>}}]
    rotationLimits [{limitMax=<2 20 0>, limitMin=<-2 -50 0>}, {limitMax=<2 15 0>, limitMin=<-2 -50 0>}, {limitMax=<2 3 0>, limitMin=<-2 -50 0>}]
    rotationOffsets [<0 0 0>, <0 0 0>, <0 0 0>]
    trailBone {boneName=dyn_r_strap_front_02}
    trailRelaxationSpeed {editorCurveData {keys [<at 0 35 RCIM_Linear>,<at 1 5 RCIM_Linear>]}}
   }
   n45 = Trail controller - Bone: dyn_l_strap_front_02(ComponentPose=@n44, Alpha=1) {
    bLimitRotation=true
    bUsePlanarLimit=true
    chainLength=3
    planarLimits [{drivingBone {boneName=dyn_strap_03}, planeTransform {location=<0 0 0>, rotation=<rot 0 0 0>, scale=<1 1 1>}}]
    rotationLimits [{limitMax=<2 10 0>, limitMin=<-2 -50 0>}, {limitMax=<2 10 0>, limitMin=<-2 -50 0>}, {limitMax=<2 3 0>, limitMin=<-2 -50 0>}]
    rotationOffsets [<0 0 0>, <0 0 0>, <0 0 0>]
    trailBone {boneName=dyn_l_strap_front_02}
    trailRelaxationSpeed {editorCurveData {keys [<at 0 35 RCIM_Linear>,<at 1 5 RCIM_Linear>]}}
   }
   n46 = Trail controller - Bone: dyn_little_strap_l_02(ComponentPose=@n45, Alpha=1) {
    bLimitRotation=true
    bUsePlanarLimit=true
    chainLength=3
    planarLimits [{drivingBone {boneName=dyn_strap_03}, planeTransform {location=<0 0 0>, rotation=<rot 0 0 0>, scale=<1 1 1>}}]
    rotationLimits [{limitMax=<2 20 0>, limitMin=<-2 -50 0>}, {limitMax=<2 15 0>, limitMin=<-2 -50 0>}, {limitMax=<2 3 0>, limitMin=<-2 -50 0>}]
    rotationOffsets [<0 0 0>, <0 0 0>, <0 0 0>]
    trailBone {boneName=dyn_little_strap_l_02}
    trailRelaxationSpeed {editorCurveData {keys [<at 0 35 RCIM_Linear>,<at 1 5 RCIM_Linear>]}}
   }
   n47 = Trail controller - Bone: dyn_little_strap_r_02(ComponentPose=@n46, Alpha=1) {
    bLimitRotation=true
    bUsePlanarLimit=true
    chainLength=3
    planarLimits [{drivingBone {boneName=dyn_strap_03}, planeTransform {location=<0 0 0>, rotation=<rot 0 0 0>, scale=<1 1 1>}}]
    rotationLimits [{limitMax=<2 0 0>, limitMin=<-2 -50 0>}, {limitMax=<2 0 0>, limitMin=<-2 -50 0>}, {limitMax=<2 0 0>, limitMin=<-2 -50 0>}]
    rotationOffsets [<0 0 0>, <0 0 0>, <0 0 0>]
    trailBone {boneName=dyn_little_strap_r_02}
    trailRelaxationSpeed {editorCurveData {keys [<at 0 35 RCIM_Linear>,<at 1 5 RCIM_Linear>]}}
   }
   n48 = Blend Poses by bool(BlendPose_0=Component To Local(ComponentPose=@n35), BlendPose_1=Component To Local(ComponentPose=@n47), bActiveValue=Get bRigidBodyEnabled()) {blendType=Linear}
   Output Pose(Result=Component To Local(ComponentPose=Local To Component(LocalPose=@n48))) {layerGroup=None}
  ]}
 EventGraph {nodes [
   Event BlueprintUpdateAnimation()
   > Sequence!(execute) -> (then_0, then_1, then_2:Exec)
   > Set Is Skydive Floating(Is Skydive Floating:double="To Float (Boolean)"(InBool=Get bIsSkyDiving())) -> (Output_Get:double)
   > n1 = Gravity Override State Function DEPRECATED(CharacterPartAnimInstance=Self-Reference(), GameTimeInSeconds=GetGameTimeInSeconds(), GravityOverrideParams=Get Gravity Override Params())
   > Set GravityOverride(GravityOverride=@n1)
   n2 = Event OnWeaponChanged()
   n3 = GetPartSkeletalMeshComponent()
   n4 = Get Dynamic Material Instance(execute=[@n2.then, @n11.then], SkeletalMeshComponent=@n3)
   > n5 = Get Dynamic Material Instance(SkeletalMeshComponent=@n3, ElementIndex=1)
   n6 = PickaxeItemDefinitionHasAnyMetaTag(Tags=Make Array("[0]"=(TagName="Cosmetics.MetaTags.SlicedBread")), bRequiresEquipped=true)
   > n7 = Branch(Condition=@n6)
   > n8 = Branch(Condition=Get bIsSkyDiving())
   SetScalarParameterValue(execute=[@n9.then, @n8.then, @n10.else, @n7.else], self=@n4, ParameterName=Hide Element 10, Value=0)
   > SetScalarParameterValue(self=@n4, ParameterName=Hide Element 02, Value=1)
   > SetScalarParameterValue(self=@n5, ParameterName=HT_CrunchVerts, Value=0)
   SetScalarParameterValue(execute=[@n8.else, @n10.then], self=@n4, ParameterName=Hide Element 02, Value=0)
   > SetScalarParameterValue(self=@n4, ParameterName=Hide Element 10, Value=1)
   > SetScalarParameterValue(self=@n5, ParameterName=HT_CrunchVerts, Value=1)
   n9 = Event OnBeginSkydiving()
   Event OnEndSkydiving()
   > n10 = Branch(Condition=@n6)
   n11 = Event OnCharacterCustomizationLoaded()
   TryGetPawnOwner()
  ]}
]
variables [
 Gravity Override Params {type=GravityOverrideParamsStruct}
 GravityOverride {type=Vector}
 Is Moving 2D {type=boolean}
 Is Parachute Opened {type=boolean}
 Is Skydive Dive Mode {type=boolean}
 Is Skydive Floating {type=number}
 Rigid Body Enabled {type=boolean}
]
